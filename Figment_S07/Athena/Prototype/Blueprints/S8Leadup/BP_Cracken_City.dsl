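; BP_Cracken_City is a Blueprint that picks which dynamic building foundation to enable for the Season 8 lead-up crack event.
; On server authority, once the match is ready, it finds the BP_S8LeadupEvents_7_4 actor and reads its CrackPhase.
; If the phase is still -1, it waits on OnCrackPhaseSet and unbinds after the first call.
; It then enables the New foundation if the phase meets Threshold, otherwise the Old one.
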(event EventBeginPlay
  (Networking|SwitchHasAuthority
    (:Authority
      (Class|FortAsyncActionAuthorityMatchReady|MatchReadyAsync
        (:OnReady
          (bind _outactors (Actor|GetAllActorsOfClass "/Figment_S07/Athena/Prototype/Blueprints/S8Leadup/BP_S8LeadupEvents_7_4.BP_S8LeadupEvents_7_4_C"))
          (Variables|Default|SetS8Events (Utilities|Array|Get(acopy) _outactors))
          (bind _output_get (Variables|Default|SetCrackPhase (Variables|/Figment_S07/Athena/Prototype/Blueprints/S8Leadup/BP_S8LeadupEvents_7_4.BP_S8LeadupEvents_7_4_C|GetCrackPhase (Variables|Default|GetS8Events))))
          (if (== _output_get -1)
            (Default|BindEventtoOnCrackPhaseSet (Variables|Default|GetS8Events) (EventDispatchers|CreateEvent :!event Custom|CrackPhaseSet))
            (else
              (CallFunction|ResumeSetup))))))))

(event Custom|ResumeSetup
  (if (>= (Variables|Default|GetCrackPhase) (Variables|Default|GetThreshold))
    (Class|BuildingFoundation|SetDynamicFoundationEnabled (Variables|Default|GetNew) true)
    (else
      (Class|BuildingFoundation|SetDynamicFoundationEnabled (Variables|Default|GetOld) true))))

(event Custom|CrackPhaseSet
  (params (CrackPhase :type Integer))
  (Variables|Default|SetCrackPhase CrackPhase)
  (Default|UnbindEventfromOnCrackPhaseSet (Variables|Default|GetS8Events) (EventDispatchers|CreateEvent :!event Custom|CrackPhaseSet))
  (CallFunction|ResumeSetup))

(fn ConstructionScript)
