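; Crack effect actor for the S8 lead-up event that spawns particle systems at stored locations once the crack phase reaches a threshold.
; On authority, InitializeAshenEffects reads the phase from the S8 events object, waiting on OnCrackPhaseSet if it is -1.
; Meeting the threshold sets SpawnEffects, whose OnRep creates the particle components on remote machines only.
; EditorEffects destroys stored components and adds new ones at the same locations.
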(event EventBeginPlay)

(event Custom|InitializeAshenEffects
  (params (NewParam :type BP%20S8Leadup%20Events%207%204%20Object%20Reference))
  (Networking|SwitchHasAuthority
    (:Authority
      (bind _output_get (Variables|Default|SetS8Events NewParam))
      (bind _output_get_2 (Variables|Default|SetCrackPhase (Variables|/Figment_S07/Athena/Prototype/Blueprints/S8Leadup/BP_S8LeadupEvents_7_4.BP_S8LeadupEvents_7_4_C|GetCrackPhase _output_get)))
      (if (== _output_get_2 -1)
        (Default|BindEventtoOnCrackPhaseSet (Variables|Default|GetS8Events) (EventDispatchers|CreateEvent :!event Custom|CrackPhaseSet))
        (else
          (CallFunction|ResumeSetup))))))

(event Custom|CrackPhaseSet
  (params (CrackPhase :type Integer))
  (Variables|Default|SetCrackPhase CrackPhase)
  (Default|UnbindEventfromOnCrackPhaseSet (Variables|Default|GetS8Events) (EventDispatchers|CreateEvent :!event Custom|CrackPhaseSet))
  (CallFunction|ResumeSetup))

(event Custom|ResumeSetup
  (if (>= (Variables|Default|GetCrackPhase) (ScalableFloat|GetValueasInteger (Variables|Default|GetCrackPhaseThreshold)))
    (Variables|Default|SetSpawnEffects true)))

(event Custom|EditorEffects
  (Utilities|FlowControl|Sequence
    (:then_0
      (for _array_element (Variables|Default|GetEffectsArray)
        (Components|DestroyComponent _array_element)))
    (:then_1
      (for _array_element_2 (Variables|Default|GetLocation)
        (bind _returnvalue (AddComponent|NODE_AddParticleSystemComponent-0 :RelativeTransform_Location _array_element_2 :RelativeTransform_Rotation "0, 0, 0" :RelativeTransform_Scale "1.000000,1.000000,1.000000"))
        (Utilities|Array|Add (Variables|Default|GetEffectsArray) _returnvalue)))))

(event Custom|Effects
  (for _array_element (Variables|Default|GetLocation)
    (bind _returnvalue (AddComponent|NODE_AddParticleSystemComponent-0 :RelativeTransform_Location _array_element :RelativeTransform_Rotation "0, 0, 0" :RelativeTransform_Scale "1.000000,1.000000,1.000000"))
    (Utilities|Array|Add (Variables|Default|GetEffectsArray) _returnvalue)))

(fn OnRep_SpawnEffects
  (Networking|SwitchHasAuthority
    (:Remote
      (if (Variables|Default|GetSpawnEffects)
        (CallFunction|Effects)))))

(fn ConstructionScript)
