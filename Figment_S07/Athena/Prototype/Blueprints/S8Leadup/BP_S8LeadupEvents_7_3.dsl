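; Season 8 lead-up event actor that triggers earthquake tremors timed against the "EQ" calendar event.
; On authority BeginPlay it compares event progress with the Quakes ratios and sets a timer for the next one if under an hour away.
; The timer increments a RepNotify counter, whose OnRep plays the tremor GameplayCue locally on the client's pawn.
; Test events regenerate Quakes from a jittered curve, print times and repeat the scheduling.
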(event Custom|TestOdds
  (Utilities|Array|Clear (Variables|Default|GetQuakes))
  (for _index (range (Variables|Default|GetNumberOfEvents))
    (Utilities|Array|SetArrayElem (Variables|Default|GetQuakes) _index (Math|Curves|GetFloatValue (Variables|Default|GetCurve) (/ (Math|Random|RandomFloatinRange (- (Math|Conversions|ToFloat(Integer) _index) 0.4) (+ (Math|Conversions|ToFloat(Integer) _index) 0.4)) 1e+02)) true)))

(event Custom|TestShake
  (Variables|Default|SetRepNotify (+ (Variables|Default|GetRepNotify) 1)))

(event Custom|PrintTimes
  (for _array_element (Variables|Default|GetQuakes)
    (Development|PrintString (Utilities|String|ToString(Float) (Math|Float|float*float _array_element 24.0 7.0)))))

(event Custom|MulticastShake
  (Networking|SwitchHasAuthority
    (:Remote
      (GameplayCue|ExecuteGameplayCue%28Burst%29Local (Pawn|GetControlledPawn (Utilities|Array|Get(acopy) (Fortnite|GetLocalFortPlayerControllers true true))) (Variables|Default|GetEarthquakeTremorsTag) (Ability|GameplayCue|MakeGameplayCueParameters :NormalizedMagnitude 0.0 :RawMagnitude 0.0 :EffectContext "()" :AggregatedSourceTags "(GameplayTags=)" :AggregatedTargetTags "(GameplayTags=)" :Location "0,0,0" :Normal "0,0,0" :GameplayEffectLevel 1 :AbilityLevel 1 :bReplicateLocationWhenUsingMinimalRepProxy false)))))

(event Custom|TestEQ
  (params (NewParam :type Float%20%28double-precision%29))
  (Variables|Default|SetRatio NewParam)
  (bind _array_element (Utilities|Array|ForEachLoopwithBreak (Variables|Default|GetQuakes))
    (:LoopBody
      (if (not (> (Variables|Default|GetRatio) _array_element))
        (CallFunction|Break)
        (bind _output_get (Variables|Default|SetTimerLength (* (- _array_element (Variables|Default|GetRatio)) (Math|Timespan|GetTotalSeconds (Math|Timespan|Timespan+Timespan (stale EventBeginPlay _timeuntilend) (stale EventBeginPlay _timesincebegin))))))
        (if (< _output_get 3.6e+03)
          (Utilities|Time|SetTimerbyEvent (EventDispatchers|CreateEvent :!event Custom|TestShake) (Variables|Default|GetTimerLength)))))))

(event EventBeginPlay
  (Networking|SwitchHasAuthority
    (:Authority
      (bind (_eventname _timeuntilend _timesincebegin _timespanratio) (MCPContext|WatchCalendarEvent "EQ")
        (:OnEventActive
          _)
        (:OnEventUpdated
          _))
      (Variables|Default|SetRatio _timespanratio)
      (bind _array_element (Utilities|Array|ForEachLoopwithBreak (Variables|Default|GetQuakes))
        (:LoopBody
          (if (not (> (Variables|Default|GetRatio) _array_element))
            (CallFunction|Break)
            (bind _output_get (Variables|Default|SetTimerLength (* (- _array_element (Variables|Default|GetRatio)) (Math|Timespan|GetTotalSeconds (Math|Timespan|Timespan+Timespan _timeuntilend _timesincebegin)))))
            (if (< _output_get 3.6e+03)
              (Utilities|Time|SetTimerbyEvent (EventDispatchers|CreateEvent :!event Custom|TestShake) (Variables|Default|GetTimerLength)))))))))

(event Custom|break
  (bind _array_element (Utilities|Array|ForEachLoopwithBreak (Variables|Default|GetQuakes))
    (:LoopBody
      (if (not (> (Variables|Default|GetRatio) _array_element))
        (CallFunction|Break)
        (bind _output_get (Variables|Default|SetTimerLength (* (- _array_element (Variables|Default|GetRatio)) (Math|Timespan|GetTotalSeconds (Math|Timespan|Timespan+Timespan (stale EventBeginPlay _timeuntilend) (stale EventBeginPlay _timesincebegin))))))
        (if (< _output_get 3.6e+03)
          (Utilities|Time|SetTimerbyEvent (EventDispatchers|CreateEvent :!event Custom|TestShake) (Variables|Default|GetTimerLength)))))))

(event Custom|DoShake
  (CallFunction|MulticastShake))

(event Collision|EventActorBeginOverlap
  (params (OtherActor :type Actor%20Object%20Reference)))

(event EventTick
  (params (DeltaSeconds :type Float%20%28single-precision%29)))

(fn OnRep_RepNotify
  (Networking|SwitchHasAuthority
    (:Remote
      (GameplayCue|ExecuteGameplayCue%28Burst%29Local (Pawn|GetControlledPawn (Utilities|Array|Get(acopy) (Fortnite|GetLocalFortPlayerControllers true true))) (Variables|Default|GetEarthquakeTremorsTag) (Ability|GameplayCue|MakeGameplayCueParameters :NormalizedMagnitude 0.0 :RawMagnitude 0.0 :EffectContext "()" :AggregatedSourceTags "(GameplayTags=)" :AggregatedTargetTags "(GameplayTags=)" :Location "0,0,0" :Normal "0,0,0" :GameplayEffectLevel 1 :AbilityLevel 1 :bReplicateLocationWhenUsingMinimalRepProxy false)))))

(fn ConstructionScript)
