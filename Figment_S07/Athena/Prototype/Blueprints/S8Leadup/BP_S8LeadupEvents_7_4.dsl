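; Season 8 lead-up event manager that stages earthquakes and advancing landscape fissures.
; On match ready, calendar events EQ, EQ2 and EQH set the crack phase and schedule timers for upcoming quakes and crack advances.
; RepNotify variables drive client tremor GameplayCues and a timeline animating the landscape MPC's CrackProgression and crack-spline particles.
; The server applies damage along active splines and box-clears buildings at set progression points; refreshes handle collision and replay resets.
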(event Custom|TestOdds
  (Utilities|Array|Clear (Variables|Default|GetQuakes))
  (for _index (range (Variables|Default|GetNumberOfEvents))
    (Utilities|Array|SetArrayElem (Variables|Default|GetQuakes) _index (Math|Curves|GetFloatValue (Variables|Default|GetCurve) (/ (Math|Random|RandomFloatinRange (- (Math|Conversions|ToFloat(Integer) _index) 0.4) (+ (Math|Conversions|ToFloat(Integer) _index) 0.4)) 1e+02)) true)))

(event Custom|PrintTimes
  (for _array_element (Variables|Default|GetQuakes)
    (Development|PrintString (Utilities|String|ToString(Float) (Math|Float|float*float _array_element 24.0 7.0)))))

(event Custom|TestShake
  (Variables|Default|SetRepNotify (+ (Variables|Default|GetRepNotify) 1)))

(event Custom|SetCurrentProgressionPhase
  (params (Progression :type Float%20%28double-precision%29))
  ;[1 1 1] Use these to test the fissures
  (Variables|Default|SetCrackPhase (Math|Float|Truncate Progression))
  (Utilities|FlowControl|Delay 0.1)
  (CallFunction|CallProgressCrack))

(event Custom|MulticastShake
  (Networking|SwitchHasAuthority
    (:Remote
      (GameplayCue|ExecuteGameplayCue%28Burst%29Local (Pawn|GetControlledPawn (Utilities|Array|Get(acopy) (Fortnite|GetLocalFortPlayerControllers true true))) (Variables|Default|GetEarthquakeTremorsTag) (Ability|GameplayCue|MakeGameplayCueParameters :NormalizedMagnitude 0.0 :RawMagnitude 0.0 :EffectContext "()" :AggregatedSourceTags "(GameplayTags=)" :AggregatedTargetTags "(GameplayTags=)" :Location "0,0,0" :Normal "0,0,0" :GameplayEffectLevel 1 :AbilityLevel 1 :bReplicateLocationWhenUsingMinimalRepProxy false)))))

(event Custom|TestCrackProgression
  (params (Amount :type Float%20%28double-precision%29))
  ;[1 1 1] Use these to test the fissures
  (CallFunction|CallProgressCrack))

(event Custom|TestCPP
  (params (CrackProgPerc :type Float%20%28double-precision%29) (TotalCalSec :type Float%20%28double-precision%29))
  (bind _progressionhour (Variables|Default|GetProgressionHour))
  ;[1 1 1] Use these to test the fissures
  (Variables|Default|SetProgressionHour CrackProgPerc)
  (Variables|Default|SetTotalCalendarSeconds TotalCalSec)
  ;[1 1 1] Crack calendar trigger
  (bind (_array_element _array_index) (Utilities|Array|ForEachLoopwithBreak (Variables|Default|GetFissureHour))
    (:LoopBody
      (if (not (> (Variables|Default|GetProgressionHour) _array_element))
        (Development|PrintString (Utilities|String|ToString(Float) (- _array_element _progressionhour)))
        (CallFunction|BreakFissuresLoop)
        (Utilities|Time|SetTimerbyEvent (delegate Custom|CallProgressCrack) (* (- _array_element _progressionhour) 3.6e+03))
        (Variables|Default|SetCrackPhase _array_index)))
    (:Completed
      ;[1 1 1] If we are past the fissure timeline, set crack phase to 27 (1 past max)
      (if (== (Variables|Default|GetCrackPhase) -1)
        (Variables|Default|SetCrackPhase 27)))))

(event Custom|TestEQ
  (params (NewParam :type Float%20%28double-precision%29))
  (Variables|Default|SetRatio NewParam)
  (bind _array_element (Utilities|Array|ForEachLoopwithBreak (Variables|Default|GetQuakes))
    (:LoopBody
      (if (not (> (Variables|Default|GetRatio) _array_element))
        (CallFunction|Break)
        (bind _output_get (Variables|Default|SetTimerLength (* (- _array_element (Variables|Default|GetRatio)) (Math|Timespan|GetTotalSeconds (Math|Timespan|Timespan+Timespan (stale EventBeginPlay _timeuntilend) (stale EventBeginPlay _timesincebegin))))))
        (if (< _output_get 3.6e+03)
          (Utilities|Time|SetTimerbyEvent (EventDispatchers|CreateEvent :!event Custom|TestShake) (Variables|Default|GetTimerLength)))))))

(event EventBeginPlay
  (bind _progressionhour (Variables|Default|GetProgressionHour))
  ;[1 1 1] Crack calendar trigger
  (Class|FortAsyncActionAuthorityMatchReady|MatchReadyAsync
    (:OnReady
      (Utilities|FlowControl|Sequence
        (:then_0
          (bind (_eventname _timeuntilend _timesincebegin _timespanratio) (MCPContext|WatchCalendarEvent "EQ")
            (:OnEventActive
              _)
            (:OnEventUpdated
              _))
          (Variables|Default|SetRatio _timespanratio)
          (bind _array_element (Utilities|Array|ForEachLoopwithBreak (Variables|Default|GetQuakes))
            (:LoopBody
              (if (not (> (Variables|Default|GetRatio) _array_element))
                (CallFunction|Break)
                (bind _output_get (Variables|Default|SetTimerLength (* (- _array_element (Variables|Default|GetRatio)) (Math|Timespan|GetTotalSeconds (Math|Timespan|Timespan+Timespan _timeuntilend _timesincebegin)))))
                (if (< _output_get 3.6e+03)
                  (Utilities|Time|SetTimerbyEvent (EventDispatchers|CreateEvent :!event Custom|TestShake) (Variables|Default|GetTimerLength)))))))
        (:then_1
          (bind (_biseventactive _timesincebegin_2 _timeuntilend_2) (CalendarEvents|GetCalendarEventInformation "EQ2"))
          (if _biseventactive
            (Variables|Default|SetProgressionHour (Math|Timespan|GetTotalHours _timesincebegin_2))
            (Variables|Default|SetTotalCalendarSeconds (Math|Timespan|GetTotalSeconds (Math|Timespan|Timespan+Timespan _timesincebegin_2 _timeuntilend_2)))
            (bind (_array_element_2 _array_index) (Utilities|Array|ForEachLoopwithBreak (Variables|Default|GetFissureHour))
              (:LoopBody
                (if (not (> (Variables|Default|GetProgressionHour) _array_element_2))
                  (Development|PrintString (Utilities|String|ToString(Float) (- _array_element_2 _progressionhour)))
                  (CallFunction|BreakFissuresLoop)
                  (Utilities|Time|SetTimerbyEvent (delegate Custom|CallProgressCrack) (* (- _array_element_2 _progressionhour) 3.6e+03))
                  (Variables|Default|SetCrackPhase _array_index)))
              (:Completed
                ;[1 1 1] If we are past the fissure timeline, set crack phase to 27 (1 past max)
                (if (== (Variables|Default|GetCrackPhase) -1)
                  (Variables|Default|SetCrackPhase 27))))
            (else
              (bind _biseventactive_2 (CalendarEvents|GetCalendarEventInformation "EQH"))
              (if _biseventactive_2
                (Variables|Default|SetCrackPhase 27)
                (else
                  (Variables|Default|SetCrackPhase 0))))))
        (:then_2
          (bind _asheneffect (Variables|Default|GetAshenEffect)
            (:then
              (Class|BPCrackenEffect|InitializeAshenEffects _asheneffect self))
            (:else)))))))

(event Custom|break
  (bind _array_element (Utilities|Array|ForEachLoopwithBreak (Variables|Default|GetQuakes))
    (:LoopBody
      (if (not (> (Variables|Default|GetRatio) _array_element))
        (CallFunction|Break)
        (bind _output_get (Variables|Default|SetTimerLength (* (- _array_element (Variables|Default|GetRatio)) (Math|Timespan|GetTotalSeconds (Math|Timespan|Timespan+Timespan (stale EventBeginPlay _timeuntilend) (stale EventBeginPlay _timesincebegin))))))
        (if (< _output_get 3.6e+03)
          (Utilities|Time|SetTimerbyEvent (EventDispatchers|CreateEvent :!event Custom|TestShake) (Variables|Default|GetTimerLength)))))))

(event Custom|BreakFissuresLoop
  (bind _progressionhour (Variables|Default|GetProgressionHour))
  ;[1 1 1] Crack calendar trigger
  (bind (_array_element _array_index) (Utilities|Array|ForEachLoopwithBreak (Variables|Default|GetFissureHour))
    (:LoopBody
      (if (not (> (Variables|Default|GetProgressionHour) _array_element))
        (Development|PrintString (Utilities|String|ToString(Float) (- _array_element _progressionhour)))
        (CallFunction|BreakFissuresLoop)
        (Utilities|Time|SetTimerbyEvent (delegate Custom|CallProgressCrack) (* (- _array_element _progressionhour) 3.6e+03))
        (Variables|Default|SetCrackPhase _array_index)))
    (:Completed
      ;[1 1 1] If we are past the fissure timeline, set crack phase to 27 (1 past max)
      (if (== (Variables|Default|GetCrackPhase) -1)
        (Variables|Default|SetCrackPhase 27)))))

(event Custom|DoShake
  (CallFunction|MulticastShake))

(event Collision|EventActorBeginOverlap
  (params (OtherActor :type Actor%20Object%20Reference)))

(event EventTick
  (params (DeltaSeconds :type Float%20%28single-precision%29)))

(event Custom|CallProgressCrack
  (CallFunction|TestShake)
  (Variables|Default|SetCurrentProgression (* (Variables|Default|GetCrackPhase) (Variables|Default|GetCrackMultiplier)))
  (for (_array_element _array_index) (Variables|Default|GetProgressionComparisonArray)
    (if (Math|Float|NearlyEqual(Float) _array_element (Variables|Default|GetCurrentProgression) 45.0)
      (Variables|Default|SetBuildingNum _array_index))))

(event Custom|KillBuildingTarget
  (switch int (Variables|Default|GetBuildingNum)
    (:0
      (Variables|Default|SetBoxPos "102200.000000,6656.000000,0.000000")
      (Variables|Default|SetBoxExt "400.000000,3000.000000,100.000000"))
    (:1
      (Variables|Default|SetBoxPos "97280.000000,-57856.000000,1536.000000")
      (Variables|Default|SetBoxExt "500.000000,500.000000,200.000000"))
    (:2
      ; OG S07 Extraction Fix - GameplayCue.Athena.Destruction.Large no longer existed.
      (Ability|ExecuteGameplayCueReplicated (Variables|Default|GetDestructionLargeTag))))
  (bind _outactors (Collision|BoxOverlapActors (Variables|Default|GetBoxPos) (Variables|Default|GetBoxExt) (Utilities|Array|MakeArray "ObjectTypeQuery1") :!class /Script/Engine.KismetSystemLibrary))
  (for _array_element _outactors
    (bind _asbuilding_actor (Utilities|Casting|CastToBuildingActor _array_element))
    (Utilities|Array|AddUnique (Variables|Default|GetActorsToDestroy) _asbuilding_actor))
  (Utilities|Time|SetTimerbyEvent (EventDispatchers|CreateEvent :!event Custom|DestroyActors) 0.001))

(event Custom|ProgressCrackPhase
  (params (CrackAmount :type Float%20%28double-precision%29))
  (Components|Timeline|SetPlayRate (Variables|BP_S8LeadupEvents_7_4|GetAdvanceCracks) (/ 1.0 (Variables|Default|GetCrackProgressionSeconds)))
  (bind (_direction _newtrack_0) (Timeline|AdvanceCracks PlayFromStart)
    (:Update
      (bind _output_get (Variables|Default|SetProgressionCalculation (+ (* (Variables|Default|GetCrackProgressionDistance) _newtrack_0) (Variables|Default|GetCurrentProgression))))
      (Rendering|Material|SetScalarParameterValue "/Game/Athena/Environments/Landscape/MPC/MPC_Landscape_S07.MPC_Landscape_S07" "CrackProgression" _output_get :!class /Script/Engine.KismetMaterialLibrary)
      (Utilities|FlowControl|Sequence
        (:then_0
          (Networking|SwitchHasAuthority
            (:Authority
              ;[1 1 1] if the phase is one where a building would be destroyed, pick a building location
              (if (> (Variables|Default|GetBuildingNum) -1)
                (if (Math|Float|NearlyEqual(Float) (Variables|Default|GetProgressionCalculation) (Utilities|Array|Get(acopy) (Variables|Default|GetProgressionComparisonArray) (Variables|Default|GetBuildingNum)) 2.0)
                  (Utilities|FlowControl|DoOnce false
                    (:Completed
                      (CallFunction|KillBuildingTarget))))))
            (:Remote
              (for _array_element (Variables|Default|GetSplineCrackBPs)
                (Class|BPSplineMeshCracks73|MoveParticleEffect _array_element)))))
        (:then_1
          (Networking|SwitchHasAuthority
            (:Authority
              (if (not (Utilities|Time|IsValidTimerHandle (Variables|Default|GetDamageTimer)))
                (CallFunction|DamageObjsinAOE)
                (bind _returnvalue (Utilities|Time|SetTimerbyEvent (delegate Custom|DamageObjsInAOE) 0.04 true))
                (Variables|Default|SetDamageTimer _returnvalue)))))))
    (:Finished
      (Development|PrintString "Hello")
      (Variables|Default|SetSetEnableCollision (Math|Random|RandomFloat)))))

(event Custom|DestroyActors
  (bind _actorstodestroy (Variables|Default|GetActorsToDestroy))
  (bind _returnvalue (Utilities|Array|LastIndex _actorstodestroy))
  (bind _output (Utilities|Array|Get(acopy) _actorstodestroy _returnvalue))
  (if (> (Utilities|Array|LastIndex (Variables|Default|GetActorsToDestroy)) 0)
    (Utilities|Time|SetTimerbyEvent (EventDispatchers|CreateEvent :!event Custom|DestroyActors) 0.001)
    (Utilities|IsValid _output
      (:Is%20Valid
        (Actor|DestroyActor _output)
        (Utilities|Array|RemoveIndex (Variables|Default|GetActorsToDestroy) _returnvalue))
      (:Is%20Not%20Valid
        (Utilities|Array|RemoveIndex (Variables|Default|GetActorsToDestroy) _returnvalue)))))

(event Custom|DamageObjsInAOE
  (for _array_element (Variables|Default|GetSplineCrackBPs)
    (Class|BPSplineMeshCracks73|MoveDamage _array_element)))

(fn OnRep_CrackPhase
  (bind _crackmultiplier (Variables|Default|GetCrackMultiplier))
  (bind _crackphase (Variables|Default|GetCrackPhase))
  (bind _returnvalue (* _crackphase _crackmultiplier))
  (Default|CallOnCrackPhaseSet (- _crackphase 1))
  (Development|PrintString (Utilities|String|ToString(Float) _returnvalue))
  (Rendering|Material|SetScalarParameterValue "/Game/Athena/Environments/Landscape/MPC/MPC_Landscape_S07.MPC_Landscape_S07" "CrackProgression" _returnvalue :!class /Script/Engine.KismetMaterialLibrary) ; Reset first
  (CallFunction|RefreshSplineCracks (Variables|Default|GetCrackPhase)))

(fn OnRep_CrackProgressionPercent)

(fn OnRep_CurrentProgression
  (CallFunction|ProgressCrackPhase 0.0))

(fn OnRep_RepNotify
  (Networking|SwitchHasAuthority
    (:Remote
      (Utilities|FlowControl|Sequence
        (:then_0
          (bind _array_element (Utilities|Array|ForEachLoopwithBreak (Variables|Default|GetSplineCrackBPs))
            (:LoopBody
              (if (Math|Float|NearlyEqual(Float) (Variables|/Figment_S07/Athena/Prototype/Blueprints/S8Leadup/BP_SplineMesh_Cracks_7_3.BP_SplineMesh_Cracks_7_3_C|GetProgressOffset _array_element) (Variables|Default|GetCurrentProgression) 4.0)
                (Variables|Default|SetCurrentCrack _array_element)))))
        (:then_1
          (GameplayCue|ExecuteGameplayCue%28Burst%29Local (Pawn|GetControlledPawn (Utilities|Array|Get(acopy) (Fortnite|GetLocalFortPlayerControllers true true))) (Variables|Default|GetEarthquakeTremorsTag) (Ability|GameplayCue|MakeGameplayCueParameters :NormalizedMagnitude 0.0 :RawMagnitude 0.0 :EffectContext "()" :AggregatedSourceTags "(GameplayTags=)" :AggregatedTargetTags "(GameplayTags=)" :Location "0,0,0" :Normal "0,0,0" :EffectCauser (Variables|Default|GetCurrentCrack) :GameplayEffectLevel 1 :AbilityLevel 1 :bReplicateLocationWhenUsingMinimalRepProxy false)))))))

(fn OnRep_SetEnableCollision
  (bind _damagetimer (Variables|Default|GetDamageTimer))
  (CallFunction|RefreshSplineCracks (Variables|Default|GetCrackPhase))
  (for _array_element (Variables|Default|GetSplineCrackBPs)
    (Components|Activation|Deactivate (Variables|/Figment_S07/Athena/Prototype/Blueprints/S8Leadup/BP_SplineMesh_Cracks_7_3.BP_SplineMesh_Cracks_7_3_C|GetEffectRight _array_element))
    (Class|BPSplineMeshCracks73|EnableCollision _array_element))
  (if (Utilities|Time|IsValidTimerHandle _damagetimer)
    (Utilities|Time|ClearandInvalidateTimerbyHandle _damagetimer)))

(fn RefreshSplineCracks
  (params (LocalCrackPhase :type Integer))
  (LocalVariables|SetLocalPhase LocalCrackPhase)
  (Utilities|Array|Clear (Variables|Default|GetSplineCrackBPs))
  (bind _outactors (Actor|GetAllActorsOfClass "/Figment_S07/Athena/Prototype/Blueprints/S8Leadup/BP_SplineMesh_Cracks_7_3.BP_SplineMesh_Cracks_7_3_C"))
  (for _array_element _outactors
    (bind _output_get (LocalVariables|SetCurrentStephase (Math|Float|Truncate (/ (Variables|/Figment_S07/Athena/Prototype/Blueprints/S8Leadup/BP_SplineMesh_Cracks_7_3.BP_SplineMesh_Cracks_7_3_C|GetProgressOffset _array_element) (Variables|Default|GetCrackMultiplier)))))
    (if (< _output_get (LocalVariables|GetLocalPhase))
      (Class|BPSplineMeshCracks73|EnableCollision _array_element)
      (Class|BPSplineMeshCracks73|CleanupExtraComponents _array_element)
      (elif (== (LocalVariables|GetCurrentStephase) (LocalVariables|GetLocalPhase))
        (Utilities|Array|AddUnique (Variables|Default|GetSplineCrackBPs) _array_element))))
  (if (Replay|GetIsPlayingReplay)
    (bind _advancecracks (Variables|BP_S8LeadupEvents_7_4|GetAdvanceCracks)
      (:then
        (Components|Timeline|Stop _advancecracks))
      (:else
        _))
    (for _array_element_2 (Variables|Default|GetSplineCrackBPs)
      (Class|BPSplineMeshCracks73|ResetComponentsonReplay _array_element_2))))

(fn ConstructionScript)
